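# Animation Blueprint for the face accessory mesh of the DrillArms character, combining poses copied from other character meshes.
# The base pose and curves come from the attached parent, while a per-bone layered blend takes clavicle, arm, hand, finger and robot arm bones from Body Mesh.
# When character customization loads, it caches Body Mesh from the custom parts and makes it a tick prerequisite of the owning component, ordering the pose copy.
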AnimBlueprint /BRCosmetics/Characters/Player/Male/Medium/Bodies/M_MED_PixieParts_DrillArms/Meshes/Parts/M_MED_PixieParts_DrillArms_FaceAcc_AnimBP
target_skeleton=&/M_MED_PixieParts_DrillArms_FaceAcc_Skeleton
graphs [
 AnimGraph {nodes [
   Copy Pose From Mesh() {bUseAttachedParent=true}
   n1 = Layered blend per bone(
    BasePose=Copy Pose From Mesh() {bCopyCurves=true, bUseAttachedParent=true}
    BlendPoses_0=Copy Pose From Mesh(SourceMeshComponent=Get Body Mesh() -> (Body Mesh:Skeletal Mesh Component Object Reference))
   ) {layerSetup [{branchFilters [
       {boneName=clavicle_l}
       {boneName=robot_clavicle_l}
       {boneName=clavicle_r}
       {boneName=robot_clavicle_r}
       {boneName=upperarm_l}
       {boneName=robot_upperarm_l}
       {boneName=upperarm_r}
       {boneName=robot_upperarm_r}
       {boneName=upperarm_twist_01_l}
       {boneName=upperarm_twist_02_l}
       {boneName=upperarm_twist_01_r}
       {boneName=upperarm_twist_02_r}
       {boneName=robot_lowerarm_l}
       {boneName=lowerarm_l}
       {boneName=robot_lowerarm_r}
       {boneName=lowerarm_r}
       {boneName=lowerarm_twist_01_l}
       {boneName=lowerarm_twist_02_l}
       {boneName=lowerarm_twist_01_r}
       {boneName=lowerarm_twist_02_r}
       {boneName=hand_l}
       {boneName=hand_r}
       {boneName=index_metacarpal_l}
       {boneName=index_metacarpal_r}
       {boneName=index_01_l}
       {boneName=index_01_r}
       {boneName=index_02_l}
       {boneName=index_02_r}
       {boneName=index_03_l}
       {boneName=index_03_r}
       {boneName=middle_metacarpal_l}
       {boneName=middle_metacarpal_r}
       {boneName=middle_01_l}
       {boneName=middle_01_r}
       {boneName=middle_02_l}
       {boneName=middle_02_r}
       {boneName=middle_03_l}
       {boneName=middle_03_r}
       {boneName=pinky_metacarpal_l}
       {boneName=pinky_metacarpal_r}
       {boneName=pinky_01_l}
       {boneName=pinky_01_r}
       {boneName=pinky_02_l}
       {boneName=pinky_02_r}
       {boneName=pinky_03_l}
       {boneName=pinky_03_r}
       {boneName=ring_metacarpal_l}
       {boneName=ring_metacarpal_r}
       {boneName=ring_01_l}
       {boneName=ring_01_r}
       {boneName=ring_02_l}
       {boneName=ring_02_r}
       {boneName=ring_03_l}
       {boneName=ring_03_r}
       {boneName=thumb_01_l}
       {boneName=thumb_01_r}
       {boneName=thumb_02_l}
       {boneName=thumb_02_r}
       {boneName=thumb_03_l}
       {boneName=thumb_03_r}
      ]}]}
   Output Pose(Result=@n1) {layerGroup=None}
  ]}
 EventGraph {nodes [
   n1 = Event OnCharacterCustomizationLoaded()
   > n2 = Set Body Mesh(Body Mesh:Skeletal Mesh Component Object Reference=Get Array Item with EFort Custom Part Type(Array=@n1.CustomMeshParts)) -> (Output_Get:Skeletal Mesh Component Object Reference)
   > AddTickPrerequisiteComponent!(execute, self:Actor Component Object Reference=GetOwningComponent(), PrerequisiteComponent:Actor Component Object Reference=@n2) -> (then)
   Event BlueprintUpdateAnimation()
  ]}
]
variables [Body Mesh {type=SkeletalMeshComponent}]
